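# M_BoxFog_far is a translucent, unlit material in the Volcano environment set, named for distant box fog.
# All shading comes from a single MF_BoxHeightFog function call fed a static bool set to true, output through material attributes.
# Depth testing is disabled, so the fog draws regardless of scene depth occlusion.
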Material /FigmentEnvironment/Materials/Volcano/M_BoxFog_far
graph {
 properties {
  bDisableDepthTest=true
  bUseMaterialAttributes=true
  blendMode=BLEND_Translucent
  pixelDepthOffsetMode=PDOM_Legacy
  shadingModel=MSM_Unlit
  usedShadingModels=Unlit
 }
 nodes [MaterialFunctionCall!(In=StaticBool() {value=true}) {materialFunction=&/MF_BoxHeightFog}]
}
